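; Sparks cosmetic component Blueprint whose only logic here toggles a debug cosmetic picker widget through a debug show/hide event.
; On show, if no valid widget exists, it async-loads a soft widget class to avoid a hard reference, then creates it, adds it to the viewport and passes itself as the cosmetic component.
; On hide, it removes the widget and clears the stored reference.
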(event EventBeginPlay)

(event EventDebug_BP_ShowCosmeticPicker
  (params (bShow :type Boolean))
  (if bShow
    (Utilities|IsValid (Variables|Default|GetDebugWidget)
      (:Is%20Not%20Valid
        (bind _class (Utilities|AsyncLoadClassAsset (Variables|Default|GetDebugWidgetSoftClass))
          (:Completed
            (bind _assparks_cosmetics_debug_picker_widget (Utilities|Casting|CastToSparksCosmeticsDebugPickerWidgetClass _class))
            ; This is a debug widget that is behind a cheat, so it's okay to sync load here.
            ;
            ; We use a soft class reference to the C++ class, with a default value of the WBP class in order to ensure this class isn't hard-referenced, such that it takes up memory.
            (bind _returnvalue (UserInterface|CreateWidget _assparks_cosmetics_debug_picker_widget))
            (Variables|Default|SetDebugWidget _returnvalue)
            (UserInterface|Viewport|AddtoViewport _returnvalue 0)
            (SparksCosmetics|Debug|SetCosmeticComponent _returnvalue self)))))
    (else
      (Utilities|IsValid (Variables|Default|GetDebugWidget)
        (:Is%20Valid
          (Widget|RemovefromParent (Variables|Default|GetDebugWidget))
          (Variables|Default|SetDebugWidget 0))))))
